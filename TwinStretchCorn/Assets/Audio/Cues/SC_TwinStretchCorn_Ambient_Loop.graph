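# Looping ambient sound cue for the TwinStretchCorn item, likely heard while it sits in the world as loot.
# A single looping wave player feeds the root directly, playing the TwinStretchCorn ambient loop wave at reduced volume.
# It uses the shared Fortnite mono item-ambient attenuation and loot ambient-loop sound class, with overridden concurrency to cap simultaneous instances.
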SoundCue /TwinStretchCorn/Assets/Audio/Cues/SC_TwinStretchCorn_Ambient_Loop
root=WavePlayer_1
nodes [{id=WavePlayer_1, properties {bLooping=true, soundWaveAssetPtr=/TwinStretchCorn/Assets/Audio/Waves/TwinStretchCorn_Ambient_Loop}}]
properties {
 attenuationSettings=/Game/Sounds/Attenuation/Fort_Item_Ambient_Loop_Mono_Default_Att
 bOverrideConcurrency=true
 maxDistance=1250
 soundClassObject=/Game/Sounds/FortSoundClassesAndModes/Fort_Loot_Ambient_Loop
 volumeMultiplier=0.5
 concurrencyOverrides {maxCount=2}
}
