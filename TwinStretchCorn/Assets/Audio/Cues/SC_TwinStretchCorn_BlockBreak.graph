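# Sound cue for the TwinStretchCorn block break effect, built from shield break recordings.
# A Random node picks one of five equally weighted Shield_Break waves and feeds a Modulator at the root.
# The Random node is excluded from branch culling.
# The cue uses a shield break attenuation asset, a BlockBreak concurrency set and a first-person ranged weapon fire sound class.
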SoundCue /TwinStretchCorn/Assets/Audio/Cues/SC_TwinStretchCorn_BlockBreak
root=Modulator_0
nodes [
 {id=Modulator_0, children [Random_1]}
 {
  id=Random_1
  children [WavePlayer_0,WavePlayer_4,WavePlayer_5,WavePlayer_6,WavePlayer_7]
  properties {bShouldExcludeFromBranchCulling=true, weights [1,1,1,1,1]}
 }
 {id=WavePlayer_0, properties {soundWaveAssetPtr=/TwinStretchCorn/Assets/Audio/Waves/TwinStretchCorn_Shield_Break_05}}
 {id=WavePlayer_4, properties {soundWaveAssetPtr=/TwinStretchCorn/Assets/Audio/Waves/TwinStretchCorn_Shield_Break_04}}
 {id=WavePlayer_5, properties {soundWaveAssetPtr=/TwinStretchCorn/Assets/Audio/Waves/TwinStretchCorn_Shield_Break_03}}
 {id=WavePlayer_6, properties {soundWaveAssetPtr=/TwinStretchCorn/Assets/Audio/Waves/TwinStretchCorn_Shield_Break_02}}
 {id=WavePlayer_7, properties {soundWaveAssetPtr=/TwinStretchCorn/Assets/Audio/Waves/TwinStretchCorn_Shield_Break_01}}
]
properties {
 attenuationSettings=/TwinStretchCorn/Assets/Audio/Settings/AS_TwinStretchCorn_Shield_Break
 maxDistance=14000
 priority=2
 soundClassObject=&Fort_WeaponRanged_Fire_1P_PSM
 volumeMultiplier=1
 concurrencyOverrides {maxCount=3}
 concurrencySet [/TwinStretchCorn/Assets/Audio/Settings/CS_TwinStretchCorn_BlockBreak]
}
