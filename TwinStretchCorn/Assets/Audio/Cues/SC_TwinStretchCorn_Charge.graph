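# Sound cue for the TwinStretchCorn charge sound, randomly picking one of five equally weighted projectile-forming waves.
# Nested IsLocal and IsEnemy branches route that shared random pool through per-listener modulators and attenuations.
# Local playback uses short attenuation at default modulation, enemies are boosted with attack attenuation, and other non-local sources are much quieter with short attenuation.
# A concurrency override caps simultaneous instances at three.
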SoundCue /TwinStretchCorn/Assets/Audio/Cues/SC_TwinStretchCorn_Charge
root=Branch_1
nodes [
 {id=Attenuation_1, children [Random_1], properties {attenuationSettings=/TwinStretchCorn/Assets/Audio/Settings/AS_TwinStretchCorn_Short}}
 {id=Attenuation_2, children [Random_1], properties {attenuationSettings=/TwinStretchCorn/Assets/Audio/Settings/AS_TwinStretchCorn_Attacks}}
 {id=Attenuation_3, children [Random_1], properties {attenuationSettings=&AS_TwinStretchCorn_Short}}
 {id=Branch_1, children [Modulator_1,Branch_2,Branch_2], properties {boolParameterName=IsLocal}}
 {id=Branch_2, children [Modulator_3,Modulator_4,Modulator_4], properties {boolParameterName=IsEnemy}}
 {id=Modulator_1, children [Attenuation_1]}
 {id=Modulator_3, children [Attenuation_2], properties {volumeMax=1.6, volumeMin=1.5}}
 {id=Modulator_4, children [Attenuation_3], properties {volumeMax=0.35, volumeMin=0.3}}
 {
  id=Random_1
  children [WavePlayer_0,WavePlayer_4,WavePlayer_5,WavePlayer_6,WavePlayer_7]
  properties {bShouldExcludeFromBranchCulling=true, weights [1,1,1,1,1]}
 }
 {id=WavePlayer_0, properties {soundWaveAssetPtr=/TwinStretchCorn/Assets/Audio/Waves/TwinStretchCorn_Proj_Form_01}}
 {id=WavePlayer_4, properties {soundWaveAssetPtr=/TwinStretchCorn/Assets/Audio/Waves/TwinStretchCorn_Proj_Form_02}}
 {id=WavePlayer_5, properties {soundWaveAssetPtr=/TwinStretchCorn/Assets/Audio/Waves/TwinStretchCorn_Proj_Form_03}}
 {id=WavePlayer_6, properties {soundWaveAssetPtr=/TwinStretchCorn/Assets/Audio/Waves/TwinStretchCorn_Proj_Form_04}}
 {id=WavePlayer_7, properties {soundWaveAssetPtr=/TwinStretchCorn/Assets/Audio/Waves/TwinStretchCorn_Proj_Form_05}}
]
properties {
 bOverrideConcurrency=true
 maxDistance=20000
 soundClassObject=&Fort_WeaponRanged_Fire_1P_PSM
 volumeMultiplier=0.7
 concurrencyOverrides {maxCount=3}
}
