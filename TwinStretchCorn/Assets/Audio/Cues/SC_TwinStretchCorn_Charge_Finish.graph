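# Sound cue for the finish of the TwinStretchCorn charge attack, currently playing a placeholder 1k test tone.
# A single wave player feeds the root directly, with no mixing or modulation.
# It uses the TwinStretchCorn attacks attenuation, is routed to the first-person ranged weapon fire sound class, and overrides concurrency to cap simultaneous instances.
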SoundCue /TwinStretchCorn/Assets/Audio/Cues/SC_TwinStretchCorn_Charge_Finish
root=WavePlayer_0
nodes [{id=WavePlayer_0, properties {soundWaveAssetPtr=/Game/Sounds/Fort_Test_Sounds/Tones/1k_test_tone}}]
properties {
 attenuationSettings=/TwinStretchCorn/Assets/Audio/Settings/AS_TwinStretchCorn_Attacks
 bOverrideConcurrency=true
 maxDistance=20000
 soundClassObject=&Fort_WeaponRanged_Fire_1P_PSM
 volumeMultiplier=1
 concurrencyOverrides {maxCount=3}
}
