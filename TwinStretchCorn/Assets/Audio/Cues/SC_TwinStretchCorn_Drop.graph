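# Sound cue for the TwinStretchCorn drop sound, set up as weapon foley.
# A Random node picks one of three equally weighted drop waves and feeds a Modulator at the root for pitch and volume variation.
# The cue lowers overall pitch and volume, and uses shared weapon-drop attenuation and a stop-oldest concurrency set.
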SoundCue /TwinStretchCorn/Assets/Audio/Cues/SC_TwinStretchCorn_Drop
root=Modulator_0
nodes [
 {id=Modulator_0, children [Random_1]}
 {id=Random_1, children [WavePlayer_0,WavePlayer_2,WavePlayer_4], properties {bShouldExcludeFromBranchCulling=true, weights [1,1,1]}}
 {id=WavePlayer_0, properties {soundWaveAssetPtr=/TwinStretchCorn/Assets/Audio/Waves/TwinStretchCorn_Drop_03}}
 {id=WavePlayer_2, properties {soundWaveAssetPtr=/TwinStretchCorn/Assets/Audio/Waves/TwinStretchCorn_Drop_02}}
 {id=WavePlayer_4, properties {soundWaveAssetPtr=/TwinStretchCorn/Assets/Audio/Waves/TwinStretchCorn_Drop_01}}
]
properties {
 attenuationSettings=&Fort_Weapon_Drop_Att
 maxDistance=3000
 pitchMultiplier=0.9
 soundClassObject=&Fort_WeaponFoley
 volumeMultiplier=0.5
 concurrencyOverrides {resolutionRule=PreventNew}
 concurrencySet [&Weapon_Drop_StopOldest]
}
