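# Sound cue for the TwinStretchCorn drop-land sound, classed as weapon foley and using weapon drop attenuation.
# A Random node picks between two equally weighted landing waves and feeds a Modulator at the root.
# Volume is scaled well down, and a concurrency override caps simultaneous instances, preventing new plays once the limit is reached.
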SoundCue /TwinStretchCorn/Assets/Audio/Cues/SC_TwinStretchCorn_DropLand
root=Modulator_0
nodes [
 {id=Modulator_0, children [Random_0]}
 {id=Random_0, children [WavePlayer_0,WavePlayer_1], properties {bShouldExcludeFromBranchCulling=true, weights [1,1]}}
 {id=WavePlayer_0, properties {soundWaveAssetPtr=/TwinStretchCorn/Assets/Audio/Waves/TwinStretchCorn_DropLand_01}}
 {id=WavePlayer_1, properties {soundWaveAssetPtr=/TwinStretchCorn/Assets/Audio/Waves/TwinStretchCorn_DropLand_02}}
]
properties {
 attenuationSettings=&Fort_Weapon_Drop_Att
 bOverrideConcurrency=true
 maxDistance=3000
 soundClassObject=&Fort_WeaponFoley
 volumeMultiplier=0.2
 concurrencyOverrides {maxCount=3, resolutionRule=PreventNew}
}
