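# Sound cue that plays the equip sound for the TwinStretchCorn item, classed as weapon foley.
# A Random node picks one of five equally weighted equip waves and feeds it through a Modulator, which varies each playback.
# The cue borrows the weapon reload attenuation, lowers its volume and overrides concurrency to cap simultaneous instances.
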SoundCue /TwinStretchCorn/Assets/Audio/Cues/SC_TwinStretchCorn_Equip
root=Modulator_0
nodes [
 {id=Modulator_0, children [Random_1]}
 {
  id=Random_1
  children [WavePlayer_0,WavePlayer_1,WavePlayer_2,WavePlayer_3,WavePlayer_7]
  properties {bShouldExcludeFromBranchCulling=true, weights [1,1,1,1,1]}
 }
 {id=WavePlayer_0, properties {soundWaveAssetPtr=/TwinStretchCorn/Assets/Audio/Waves/TwinStretchCorn_Equip_01}}
 {id=WavePlayer_1, properties {soundWaveAssetPtr=/TwinStretchCorn/Assets/Audio/Waves/TwinStretchCorn_Equip_02}}
 {id=WavePlayer_2, properties {soundWaveAssetPtr=/TwinStretchCorn/Assets/Audio/Waves/TwinStretchCorn_Equip_03}}
 {id=WavePlayer_3, properties {soundWaveAssetPtr=/TwinStretchCorn/Assets/Audio/Waves/TwinStretchCorn_Equip_04}}
 {id=WavePlayer_7, properties {soundWaveAssetPtr=/TwinStretchCorn/Assets/Audio/Waves/TwinStretchCorn_Equip_05}}
]
properties {
 attenuationSettings=&Fort_Weapon_Reload_Att
 bOverrideConcurrency=true
 maxDistance=2250
 soundClassObject=&Fort_WeaponFoley
 volumeMultiplier=0.6
 concurrencyOverrides {maxCount=3}
}
