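# Sound cue for the TwinStretchCorn jump hover sound, routed to the Athena player foley sound class.
# An IsLocal branch has only its True input wired, so the sound plays only when that parameter is true.
# That path applies slight volume modulation to an equally weighted random pick of six wave variations.
# The cue lowers overall volume and overrides concurrency to cap simultaneous instances at three.
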SoundCue /TwinStretchCorn/Assets/Audio/Cues/SC_TwinStretchCorn_Jump_Hover
root=Branch_0
nodes [
 {id=Branch_0, children [Modulator_0,null,null], properties {boolParameterName=IsLocal}}
 {id=Modulator_0, children [Random_1], properties {volumeMax=0.9, volumeMin=0.8}}
 {
  id=Random_1
  children [WavePlayer_0,WavePlayer_4,WavePlayer_5,WavePlayer_6,WavePlayer_7,WavePlayer_8]
  properties {bShouldExcludeFromBranchCulling=true, weights [1,1,1,1,1,1]}
 }
 {id=WavePlayer_0, properties {soundWaveAssetPtr=/TwinStretchCorn/Assets/Audio/Waves/TwinStretchCorn_Jump_Hover_01}}
 {id=WavePlayer_4, properties {soundWaveAssetPtr=/TwinStretchCorn/Assets/Audio/Waves/TwinStretchCorn_Jump_Hover_02}}
 {id=WavePlayer_5, properties {soundWaveAssetPtr=/TwinStretchCorn/Assets/Audio/Waves/TwinStretchCorn_Jump_Hover_03}}
 {id=WavePlayer_6, properties {soundWaveAssetPtr=/TwinStretchCorn/Assets/Audio/Waves/TwinStretchCorn_Jump_Hover_04}}
 {id=WavePlayer_7, properties {soundWaveAssetPtr=/TwinStretchCorn/Assets/Audio/Waves/TwinStretchCorn_Jump_Hover_05}}
 {id=WavePlayer_8, properties {soundWaveAssetPtr=/TwinStretchCorn/Assets/Audio/Waves/TwinStretchCorn_Jump_Hover_06}}
]
properties {
 bOverrideConcurrency=true
 maxDistance=2097152
 soundClassObject=&Fort_Player_Foley_Athena
 volumeMultiplier=0.7
 concurrencyOverrides {maxCount=3}
}
