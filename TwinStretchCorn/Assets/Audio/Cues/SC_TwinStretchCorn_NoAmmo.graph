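# Sound cue for the TwinStretchCorn no-ammo sound, by its name the feedback heard when that item is out of ammo.
# A short fixed delay leads into a Modulator at default settings, fed by a Random node that picks among six equally weighted wave variants.
# It borrows weapon-pickup attenuation and the loot-collect sound class.
# A concurrency override allows one instance at a time and blocks new plays while one is active.
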SoundCue /TwinStretchCorn/Assets/Audio/Cues/SC_TwinStretchCorn_NoAmmo
root=Delay_0
nodes [
 {id=Delay_0, children [Modulator_0], properties {delayMax=0.2, delayMin=0.2}}
 {id=Modulator_0, children [Random_1]}
 {
  id=Random_1
  children [WavePlayer_1,WavePlayer_4,WavePlayer_5,WavePlayer_6,WavePlayer_7,WavePlayer_8]
  properties {bShouldExcludeFromBranchCulling=true, weights [1,1,1,1,1,1]}
 }
 {id=WavePlayer_1, properties {soundWaveAssetPtr=/TwinStretchCorn/Assets/Audio/Waves/TwinStretchCorn_NoAmmo_01}}
 {id=WavePlayer_4, properties {soundWaveAssetPtr=/TwinStretchCorn/Assets/Audio/Waves/TwinStretchCorn_NoAmmo_02}}
 {id=WavePlayer_5, properties {soundWaveAssetPtr=/TwinStretchCorn/Assets/Audio/Waves/TwinStretchCorn_NoAmmo_03}}
 {id=WavePlayer_6, properties {soundWaveAssetPtr=/TwinStretchCorn/Assets/Audio/Waves/TwinStretchCorn_NoAmmo_04}}
 {id=WavePlayer_7, properties {soundWaveAssetPtr=/TwinStretchCorn/Assets/Audio/Waves/TwinStretchCorn_NoAmmo_05}}
 {id=WavePlayer_8, properties {soundWaveAssetPtr=/TwinStretchCorn/Assets/Audio/Waves/TwinStretchCorn_NoAmmo_06}}
]
properties {
 attenuationSettings=&Fort_Weapon_Pickup_Att
 bOverrideConcurrency=true
 maxDistance=3500
 soundClassObject=&Fort_Loot_Collect
 volumeMultiplier=1
 concurrencyOverrides {maxCount=1, resolutionRule=PreventNew}
}
