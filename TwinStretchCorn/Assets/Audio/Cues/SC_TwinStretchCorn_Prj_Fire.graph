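# Sound cue for TwinStretchCorn projectile fire, playing one of six equally weighted wave variations at random.
# Nested IsLocal and IsEnemy branches share that random pool through separate volume modulator and attenuation chains.
# Non-local enemy shots are boosted with Attacks attenuation, local shots use Short attenuation, and other shots are quieter with Short attenuation.
# It uses a ranged-weapon fire sound class and two concurrency sets.
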SoundCue /TwinStretchCorn/Assets/Audio/Cues/SC_TwinStretchCorn_Prj_Fire
root=Branch_2
nodes [
 {id=Attenuation_3, children [Random_1], properties {attenuationSettings=/TwinStretchCorn/Assets/Audio/Settings/AS_TwinStretchCorn_Short}}
 {id=Attenuation_4, children [Random_1], properties {attenuationSettings=/TwinStretchCorn/Assets/Audio/Settings/AS_TwinStretchCorn_Attacks}}
 {id=Attenuation_5, children [Random_1], properties {attenuationSettings=&AS_TwinStretchCorn_Short}}
 {id=Branch_2, children [Modulator_3,Branch_3,Branch_3], properties {boolParameterName=IsLocal}}
 {id=Branch_3, children [Modulator_4,Modulator_5,Modulator_5], properties {boolParameterName=IsEnemy}}
 {id=Modulator_3, children [Attenuation_5]}
 {id=Modulator_4, children [Attenuation_4], properties {volumeMax=1.8, volumeMin=1.7}}
 {id=Modulator_5, children [Attenuation_3], properties {volumeMax=0.55, volumeMin=0.5}}
 {
  id=Random_1
  children [WavePlayer_0,WavePlayer_4,WavePlayer_5,WavePlayer_6,WavePlayer_7,WavePlayer_8]
  properties {bShouldExcludeFromBranchCulling=true, weights [1,1,1,1,1,1]}
 }
 {id=WavePlayer_0, properties {soundWaveAssetPtr=/TwinStretchCorn/Assets/Audio/Waves/TwinStretchCorn_Proj_Fire_01}}
 {id=WavePlayer_4, properties {soundWaveAssetPtr=/TwinStretchCorn/Assets/Audio/Waves/TwinStretchCorn_Proj_Fire_02}}
 {id=WavePlayer_5, properties {soundWaveAssetPtr=/TwinStretchCorn/Assets/Audio/Waves/TwinStretchCorn_Proj_Fire_03}}
 {id=WavePlayer_6, properties {soundWaveAssetPtr=/TwinStretchCorn/Assets/Audio/Waves/TwinStretchCorn_Proj_Fire_04}}
 {id=WavePlayer_7, properties {soundWaveAssetPtr=/TwinStretchCorn/Assets/Audio/Waves/TwinStretchCorn_Proj_Fire_05}}
 {id=WavePlayer_8, properties {soundWaveAssetPtr=/TwinStretchCorn/Assets/Audio/Waves/TwinStretchCorn_Proj_Fire_06}}
]
properties {
 maxDistance=20000
 soundClassObject=&Fort_WeaponRanged_Fire_1P_PSM
 volumeMultiplier=1.1
 concurrencyOverrides {maxCount=3, resolutionRule=StopFarthestThenPreventNew, volumeScaleAttackTime=0.1}
 concurrencySet [/TwinStretchCorn/Assets/Audio/Settings/CS_TwinStretchCorn_Proj, /TwinStretchCorn/Assets/Audio/Settings/CS_TwinStretchCorn_MaxThree]
}
