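# Sound cue for TwinStretchCorn projectile impacts, routed to a medium explosives sound class with a three-voice concurrency override.
# A near layer mixes a random impact wave with a briefly delayed random debris wave, each with its own attenuation.
# A distance crossfade hands this near mix over to a louder-weighted random distant impact layer at long range, all under a root modulator.
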SoundCue /TwinStretchCorn/Assets/Audio/Cues/SC_TwinStretchCorn_Prj_Impact
root=Modulator_13
nodes [
 {
  id=Attenuation_0
  children [Random_2]
  properties {attenuationSettings=/TwinStretchCorn/Assets/Audio/Settings/AS_TwinStretchCorn_Impacts_Debris_Close}
 }
 {id=Attenuation_3, children [Random_1], properties {attenuationSettings=/TwinStretchCorn/Assets/Audio/Settings/AS_TwinStretchCorn_Impacts}}
 {
  id=Attenuation_4
  children [Random_0]
  properties {attenuationSettings=/TwinStretchCorn/Assets/Audio/Settings/AS_TwinStretchCorn_Impacts_Distant}
 }
 {id=Delay_0, children [Modulator_1], properties {delayMax=0.2, delayMin=0.1}}
 {
  id=DistanceCrossFade_0
  children [Mixer_0, Attenuation_4]
  properties {crossFadeInput [
    {fadeInDistanceEnd=0, fadeInDistanceStart=0, fadeOutDistanceEnd=20000, fadeOutDistanceStart=10000, volume=1}
    {fadeInDistanceEnd=20000, fadeInDistanceStart=10000, fadeOutDistanceEnd=25000, fadeOutDistanceStart=20000, volume=1.5}
   ]}
 }
 {id=Mixer_0, children [Delay_0,Modulator_0], properties {inputVolume [1,0.7]}}
 {id=Modulator_0, children [Attenuation_3], properties {pitchMax=1, pitchMin=1, volumeMax=0.85, volumeMin=0.8}}
 {id=Modulator_1, children [Attenuation_0]}
 {id=Modulator_13, children [DistanceCrossFade_0]}
 {
  id=Random_0
  children [WavePlayer_1, WavePlayer_2, WavePlayer_3, WavePlayer_9, WavePlayer_10, WavePlayer_11]
  properties {bShouldExcludeFromBranchCulling=true, weights [1,1,1,1,1,1]}
 }
 {
  id=Random_1
  children [WavePlayer_0,WavePlayer_4,WavePlayer_5,WavePlayer_6,WavePlayer_7,WavePlayer_8]
  properties {bShouldExcludeFromBranchCulling=true, weights [1,1,1,1,1,1]}
 }
 {
  id=Random_2
  children [WavePlayer_12, WavePlayer_13, WavePlayer_14, WavePlayer_15, WavePlayer_16, WavePlayer_17]
  properties {bShouldExcludeFromBranchCulling=true, weights [1,1,1,1,1,1]}
 }
 {id=WavePlayer_0, properties {soundWaveAssetPtr=/TwinStretchCorn/Assets/Audio/Waves/TwinStretchCorn_Proj_Impact_01}}
 {id=WavePlayer_1, properties {soundWaveAssetPtr=/TwinStretchCorn/Assets/Audio/Waves/TwinStretchCorn_Proj_Impact_Distant_01}}
 {id=WavePlayer_10, properties {soundWaveAssetPtr=/TwinStretchCorn/Assets/Audio/Waves/TwinStretchCorn_Proj_Impact_Distant_05}}
 {id=WavePlayer_11, properties {soundWaveAssetPtr=/TwinStretchCorn/Assets/Audio/Waves/TwinStretchCorn_Proj_Impact_Distant_06}}
 {id=WavePlayer_12, properties {soundWaveAssetPtr=/TwinStretchCorn/Assets/Audio/Waves/TwinStretchCorn_Proj_Impact_Debris_06}}
 {id=WavePlayer_13, properties {soundWaveAssetPtr=/TwinStretchCorn/Assets/Audio/Waves/TwinStretchCorn_Proj_Impact_Debris_05}}
 {id=WavePlayer_14, properties {soundWaveAssetPtr=/TwinStretchCorn/Assets/Audio/Waves/TwinStretchCorn_Proj_Impact_Debris_04}}
 {id=WavePlayer_15, properties {soundWaveAssetPtr=/TwinStretchCorn/Assets/Audio/Waves/TwinStretchCorn_Proj_Impact_Debris_03}}
 {id=WavePlayer_16, properties {soundWaveAssetPtr=/TwinStretchCorn/Assets/Audio/Waves/TwinStretchCorn_Proj_Impact_Debris_02}}
 {id=WavePlayer_17, properties {soundWaveAssetPtr=/TwinStretchCorn/Assets/Audio/Waves/TwinStretchCorn_Proj_Impact_Debris_01}}
 {id=WavePlayer_2, properties {soundWaveAssetPtr=/TwinStretchCorn/Assets/Audio/Waves/TwinStretchCorn_Proj_Impact_Distant_02}}
 {id=WavePlayer_3, properties {soundWaveAssetPtr=/TwinStretchCorn/Assets/Audio/Waves/TwinStretchCorn_Proj_Impact_Distant_03}}
 {id=WavePlayer_4, properties {soundWaveAssetPtr=/TwinStretchCorn/Assets/Audio/Waves/TwinStretchCorn_Proj_Impact_02}}
 {id=WavePlayer_5, properties {soundWaveAssetPtr=/TwinStretchCorn/Assets/Audio/Waves/TwinStretchCorn_Proj_Impact_03}}
 {id=WavePlayer_6, properties {soundWaveAssetPtr=/TwinStretchCorn/Assets/Audio/Waves/TwinStretchCorn_Proj_Impact_04}}
 {id=WavePlayer_7, properties {soundWaveAssetPtr=/TwinStretchCorn/Assets/Audio/Waves/TwinStretchCorn_Proj_Impact_05}}
 {id=WavePlayer_8, properties {soundWaveAssetPtr=/TwinStretchCorn/Assets/Audio/Waves/TwinStretchCorn_Proj_Impact_06}}
 {id=WavePlayer_9, properties {soundWaveAssetPtr=/TwinStretchCorn/Assets/Audio/Waves/TwinStretchCorn_Proj_Impact_Distant_04}}
]
properties {
 bOverrideConcurrency=true
 maxDistance=25500
 priority=2
 soundClassObject=/Game/Sounds/FortSoundClassesAndModes/Fort_Explosives_Medium_PSM
 volumeMultiplier=0.5
 concurrencyOverrides {maxCount=3, retriggerTime=1, volumeScaleAttackTime=0.1}
}
