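# Sound cue for the TwinStretchCorn projectile's water impact, routed to Fortnite's medium explosives sound class with a small concurrency limit.
# A distance crossfade blends a near layer into a distant layer as range grows, with the distant layer fading out at the far end.
# Each layer randomly picks one of five waves through its own impact attenuation asset.
# Modulators vary overall volume and slightly lower the near layer.
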SoundCue /TwinStretchCorn/Assets/Audio/Cues/SC_TwinStretchCorn_Prj_Impact_Water
root=Modulator_13
nodes [
 {id=Attenuation_3, children [Random_2], properties {attenuationSettings=/TwinStretchCorn/Assets/Audio/Settings/AS_TwinStretchCorn_Impacts}}
 {
  id=Attenuation_4
  children [Random_3]
  properties {attenuationSettings=/TwinStretchCorn/Assets/Audio/Settings/AS_TwinStretchCorn_Impacts_Distant}
 }
 {
  id=DistanceCrossFade_0
  children [Modulator_0, Attenuation_4]
  properties {crossFadeInput [
    {fadeInDistanceEnd=0, fadeInDistanceStart=0, fadeOutDistanceEnd=20000, fadeOutDistanceStart=10000, volume=1}
    {fadeInDistanceEnd=20000, fadeInDistanceStart=10000, fadeOutDistanceEnd=25000, fadeOutDistanceStart=20000, volume=1}
   ]}
 }
 {id=Modulator_0, children [Attenuation_3], properties {pitchMax=1, pitchMin=1, volumeMax=0.85, volumeMin=0.8}}
 {id=Modulator_13, children [DistanceCrossFade_0], properties {volumeMin=0.9}}
 {
  id=Random_2
  children [WavePlayer_12, WavePlayer_13, WavePlayer_14, WavePlayer_15, WavePlayer_16]
  properties {bShouldExcludeFromBranchCulling=true, weights [1,1,1,1,1]}
 }
 {
  id=Random_3
  children [WavePlayer_17, WavePlayer_18, WavePlayer_19, WavePlayer_20, WavePlayer_21]
  properties {bShouldExcludeFromBranchCulling=true, weights [1,1,1,1,1]}
 }
 {id=WavePlayer_12, properties {soundWaveAssetPtr=/TwinStretchCorn/Assets/Audio/Waves/TwinStretchCorn_Proj_Impact_Water_Layer_01}}
 {id=WavePlayer_13, properties {soundWaveAssetPtr=/TwinStretchCorn/Assets/Audio/Waves/TwinStretchCorn_Proj_Impact_Water_Layer_02}}
 {id=WavePlayer_14, properties {soundWaveAssetPtr=/TwinStretchCorn/Assets/Audio/Waves/TwinStretchCorn_Proj_Impact_Water_Layer_03}}
 {id=WavePlayer_15, properties {soundWaveAssetPtr=/TwinStretchCorn/Assets/Audio/Waves/TwinStretchCorn_Proj_Impact_Water_Layer_04}}
 {id=WavePlayer_16, properties {soundWaveAssetPtr=/TwinStretchCorn/Assets/Audio/Waves/TwinStretchCorn_Proj_Impact_Water_Layer_05}}
 {id=WavePlayer_17, properties {soundWaveAssetPtr=/TwinStretchCorn/Assets/Audio/Waves/TwinStretchCorn_Proj_Impact_Water_Layer_Distant_01}}
 {id=WavePlayer_18, properties {soundWaveAssetPtr=/TwinStretchCorn/Assets/Audio/Waves/TwinStretchCorn_Proj_Impact_Water_Layer_Distant_02}}
 {id=WavePlayer_19, properties {soundWaveAssetPtr=/TwinStretchCorn/Assets/Audio/Waves/TwinStretchCorn_Proj_Impact_Water_Layer_Distant_03}}
 {id=WavePlayer_20, properties {soundWaveAssetPtr=/TwinStretchCorn/Assets/Audio/Waves/TwinStretchCorn_Proj_Impact_Water_Layer_Distant_04}}
 {id=WavePlayer_21, properties {soundWaveAssetPtr=/TwinStretchCorn/Assets/Audio/Waves/TwinStretchCorn_Proj_Impact_Water_Layer_Distant_05}}
]
properties {
 maxDistance=25500
 priority=2
 soundClassObject=/Game/Sounds/FortSoundClassesAndModes/Fort_Explosives_Medium_PSM
 volumeMultiplier=0.5
 concurrencyOverrides {maxCount=3}
 concurrencySet [null]
}
