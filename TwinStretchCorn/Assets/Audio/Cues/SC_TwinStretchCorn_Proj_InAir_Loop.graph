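# Looping Sound Cue for the in-air flight sound of the TwinStretchCorn projectile, limited by a concurrency override.
# One of three looping waves is picked at random and shared by three paths that use the same attenuation settings.
# IsLocal and IsEnemy branches select the path, boosting enemy projectiles and heavily lowering other non-local ones.
# The result passes through smoothed Doppler and a slight random pitch shift at the root.
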SoundCue /TwinStretchCorn/Assets/Audio/Cues/SC_TwinStretchCorn_Proj_InAir_Loop
root=Modulator_0
nodes [
 {id=Attenuation_0, children [Random_1], properties {attenuationSettings=/TwinStretchCorn/Assets/Audio/Settings/AS_TwinStretchCorn_Proj_InAir}}
 {id=Attenuation_2, children [Random_1], properties {attenuationSettings=&AS_TwinStretchCorn_Proj_InAir}}
 {id=Attenuation_5, children [Random_1], properties {attenuationSettings=&AS_TwinStretchCorn_Proj_InAir}}
 {id=Branch_1, children [Modulator_2,Branch_3,Branch_3], properties {boolParameterName=IsLocal}}
 {id=Branch_3, children [Modulator_4,Modulator_5,Modulator_5], properties {boolParameterName=IsEnemy}}
 {id=Doppler_1, children [Branch_1], properties {bUseSmoothing=true}}
 {id=Modulator_0, children [Doppler_1], properties {pitchMax=1.1, pitchMin=0.9}}
 {id=Modulator_2, children [Attenuation_0]}
 {id=Modulator_4, children [Attenuation_2], properties {volumeMax=1.4, volumeMin=1.3}}
 {id=Modulator_5, children [Attenuation_5], properties {volumeMax=0.35, volumeMin=0.3}}
 {id=Random_1, children [WavePlayer_0,WavePlayer_3,WavePlayer_4], properties {bShouldExcludeFromBranchCulling=true, weights [1,1,1]}}
 {id=WavePlayer_0, properties {bLooping=true, soundWaveAssetPtr=/TwinStretchCorn/Assets/Audio/Waves/TwinStretchCorn_Proj_InAir_Loop_01}}
 {id=WavePlayer_3, properties {bLooping=true, soundWaveAssetPtr=/TwinStretchCorn/Assets/Audio/Waves/TwinStretchCorn_Proj_InAir_Loop_02}}
 {id=WavePlayer_4, properties {bLooping=true, soundWaveAssetPtr=/TwinStretchCorn/Assets/Audio/Waves/TwinStretchCorn_Proj_InAir_Loop_03}}
]
properties {
 bOverrideConcurrency=true
 maxDistance=19000
 soundClassObject=&Fort_WeaponRanged_Fire_1P_PSM
 volumeMultiplier=1.3
 concurrencyOverrides {maxCount=2, volumeScaleAttackTime=0.1}
}
