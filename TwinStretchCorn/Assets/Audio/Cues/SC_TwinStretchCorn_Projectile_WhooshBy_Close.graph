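# Sound cue for the close whoosh-by of a TwinStretchCorn projectile.
# A Random node picks one of six equally weighted whizz-by waves and feeds a Modulator at default settings for slight variation.
# It plays in the ranged-weapon projectile effects sound class, using shared bullet-whip attenuation and bullet whiz-by concurrency rules, with virtualization disabled.
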SoundCue /TwinStretchCorn/Assets/Audio/Cues/SC_TwinStretchCorn_Projectile_WhooshBy_Close
root=Modulator_0
nodes [
 {id=Modulator_0, children [Random_0]}
 {
  id=Random_0
  children [WavePlayer_0,WavePlayer_1,WavePlayer_2,WavePlayer_3,WavePlayer_4,WavePlayer_5]
  properties {bShouldExcludeFromBranchCulling=true, weights [1,1,1,1,1,1]}
 }
 {id=WavePlayer_0, properties {soundWaveAssetPtr=/TwinStretchCorn/Assets/Audio/Waves/TwinStretchCorn_Proj_WhizzBy_01}}
 {id=WavePlayer_1, properties {soundWaveAssetPtr=/TwinStretchCorn/Assets/Audio/Waves/TwinStretchCorn_Proj_WhizzBy_02}}
 {id=WavePlayer_2, properties {soundWaveAssetPtr=/TwinStretchCorn/Assets/Audio/Waves/TwinStretchCorn_Proj_WhizzBy_03}}
 {id=WavePlayer_3, properties {soundWaveAssetPtr=/TwinStretchCorn/Assets/Audio/Waves/TwinStretchCorn_Proj_WhizzBy_04}}
 {id=WavePlayer_4, properties {soundWaveAssetPtr=/TwinStretchCorn/Assets/Audio/Waves/TwinStretchCorn_Proj_WhizzBy_05}}
 {id=WavePlayer_5, properties {soundWaveAssetPtr=/TwinStretchCorn/Assets/Audio/Waves/TwinStretchCorn_Proj_WhizzBy_06}}
]
properties {
 attenuationSettings=/Game/Athena/Sounds/Weapons/BulletWhips/bullet_whip_Att.Bullet_Whip_Att
 bBypassVolumeScaleForPriority=true
 maxDistance=2097152
 priority=0.5
 soundClassObject=&Fort_WeaponRanged_Projectile_Effects
 virtualizationMode=Disabled
 volumeMultiplier=0.8
 attenuationOverrides {bAttenuate=false, stereoSpread=0}
 concurrencySet [/Game/Sounds/Concurrency_Rules/Athena_BulletWhizBy_Conc]
}
