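# Sound Cue for the distant whoosh-by of TwinStretchCorn projectiles.
# A Random node picks one of six equally weighted distant whizz-by waves, and a Modulator at the root varies the playback.
# It plays in the ranged weapon projectile effects sound class, using shared bullet-whip attenuation and a bullet whiz-by concurrency rule, with virtualization disabled.
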SoundCue /TwinStretchCorn/Assets/Audio/Cues/SC_TwinStretchCorn_Projectile_WhooshBy_Distant
root=Modulator_0
nodes [
 {id=Modulator_0, children [Random_1]}
 {
  id=Random_1
  children [WavePlayer_6, WavePlayer_7, WavePlayer_8, WavePlayer_9, WavePlayer_10, WavePlayer_11]
  properties {bShouldExcludeFromBranchCulling=true, weights [1,1,1,1,1,1]}
 }
 {id=WavePlayer_10, properties {soundWaveAssetPtr=/TwinStretchCorn/Assets/Audio/Waves/TwinStretchCorn_Proj_WhizzBy_Distant_05}}
 {id=WavePlayer_11, properties {soundWaveAssetPtr=/TwinStretchCorn/Assets/Audio/Waves/TwinStretchCorn_Proj_WhizzBy_Distant_06}}
 {id=WavePlayer_6, properties {soundWaveAssetPtr=/TwinStretchCorn/Assets/Audio/Waves/TwinStretchCorn_Proj_WhizzBy_Distant_01}}
 {id=WavePlayer_7, properties {soundWaveAssetPtr=/TwinStretchCorn/Assets/Audio/Waves/TwinStretchCorn_Proj_WhizzBy_Distant_02}}
 {id=WavePlayer_8, properties {soundWaveAssetPtr=/TwinStretchCorn/Assets/Audio/Waves/TwinStretchCorn_Proj_WhizzBy_Distant_03}}
 {id=WavePlayer_9, properties {soundWaveAssetPtr=/TwinStretchCorn/Assets/Audio/Waves/TwinStretchCorn_Proj_WhizzBy_Distant_04}}
]
properties {
 attenuationSettings=/Game/Athena/Sounds/Weapons/BulletWhips/bullet_whip_Att.Bullet_Whip_Att
 bBypassVolumeScaleForPriority=true
 maxDistance=2097152
 priority=0.5
 soundClassObject=&Fort_WeaponRanged_Projectile_Effects
 virtualizationMode=Disabled
 volumeMultiplier=1
 attenuationOverrides {bAttenuate=false, stereoSpread=0}
 concurrencySet [/Game/Sounds/Concurrency_Rules/Athena_BulletWhizBy_Conc]
}
