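# Sound cue for the TwinStretchCorn recharged sound, classed as weapon foley.
# A random node picks one of four equally weighted wave variations and feeds it through a modulator at the root.
# It plays at reduced volume with the weapon reload attenuation and a concurrency override that limits simultaneous instances.
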SoundCue /TwinStretchCorn/Assets/Audio/Cues/SC_TwinStretchCorn_Recharged
root=Modulator_0
nodes [
 {id=Modulator_0, children [Random_0]}
 {
  id=Random_0
  children [WavePlayer_4,WavePlayer_5,WavePlayer_6,WavePlayer_8]
  properties {bShouldExcludeFromBranchCulling=true, weights [1,1,1,1]}
 }
 {id=WavePlayer_4, properties {soundWaveAssetPtr=/TwinStretchCorn/Assets/Audio/Waves/TwimStretchCorn_Recharged_01}}
 {id=WavePlayer_5, properties {soundWaveAssetPtr=/TwinStretchCorn/Assets/Audio/Waves/TwimStretchCorn_Recharged_02}}
 {id=WavePlayer_6, properties {soundWaveAssetPtr=/TwinStretchCorn/Assets/Audio/Waves/TwimStretchCorn_Recharged_03}}
 {id=WavePlayer_8, properties {soundWaveAssetPtr=/TwinStretchCorn/Assets/Audio/Waves/TwimStretchCorn_Recharged_04}}
]
properties {
 attenuationSettings=&Fort_Weapon_Reload_Att
 bOverrideConcurrency=true
 maxDistance=2250
 soundClassObject=&Fort_WeaponFoley
 volumeMultiplier=0.25
 concurrencyOverrides {maxCount=3}
}
