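# Sound cue for the TwinStretchCorn reload, built from the item's Shield_Form wave variations.
# A root Random node picks one of three equally weighted wave players on each play.
# The cue uses the TwinStretchCorn attacks attenuation and a first-person ranged weapon fire sound class, and overrides concurrency to cap simultaneous instances.
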SoundCue /TwinStretchCorn/Assets/Audio/Cues/SC_TwinStretchCorn_Reload
root=Random_0
nodes [
 {id=Random_0, children [WavePlayer_1,WavePlayer_2,WavePlayer_3], properties {weights [1,1,1]}}
 {id=WavePlayer_1, properties {soundWaveAssetPtr=/TwinStretchCorn/Assets/Audio/Waves/TwinStretchCorn_Shield_Form_01}}
 {id=WavePlayer_2, properties {soundWaveAssetPtr=/TwinStretchCorn/Assets/Audio/Waves/TwinStretchCorn_Shield_Form_02}}
 {id=WavePlayer_3, properties {soundWaveAssetPtr=/TwinStretchCorn/Assets/Audio/Waves/TwinStretchCorn_Shield_Form_03}}
]
properties {
 attenuationSettings=/TwinStretchCorn/Assets/Audio/Settings/AS_TwinStretchCorn_Attacks
 bOverrideConcurrency=true
 maxDistance=20000
 soundClassObject=&Fort_WeaponRanged_Fire_1P_PSM
 volumeMultiplier=1
 concurrencyOverrides {maxCount=3}
}
