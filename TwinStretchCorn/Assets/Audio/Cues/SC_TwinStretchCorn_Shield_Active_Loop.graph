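# Sound cue for the looping active-shield audio of TwinStretchCorn, randomly choosing one of two looping waves.
# IsLocal and IsEnemy branches route that shared random pick through different volume modulators: near default for the local player, boosted for enemies, much quieter otherwise.
# The random node is excluded from branch culling.
# It uses short-range attenuation, a ranged-weapon first-person fire sound class, and a concurrency override allowing one instance.
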SoundCue /TwinStretchCorn/Assets/Audio/Cues/SC_TwinStretchCorn_Shield_Active_Loop
root=Branch_2
nodes [
 {id=Branch_1, children [Modulator_3,Modulator_2,Modulator_2], properties {boolParameterName=IsEnemy}}
 {id=Branch_2, children [Modulator_5,Branch_1,Branch_1], properties {boolParameterName=IsLocal}}
 {id=Modulator_2, children [Random_0], properties {volumeMax=0.35, volumeMin=0.3}}
 {id=Modulator_3, children [Random_0], properties {volumeMax=2.1, volumeMin=2}}
 {id=Modulator_5, children [Random_0]}
 {id=Random_0, children [WavePlayer_1,WavePlayer_2], properties {bShouldExcludeFromBranchCulling=true, weights [1,1]}}
 {id=WavePlayer_1, properties {bLooping=true, soundWaveAssetPtr=/TwinStretchCorn/Assets/Audio/Waves/TwinStretchCorn_Shield_Active_Loop_01}}
 {id=WavePlayer_2, properties {bLooping=true, soundWaveAssetPtr=/TwinStretchCorn/Assets/Audio/Waves/TwinStretchCorn_Shield_Active_Loop_02}}
]
properties {
 attenuationSettings=/TwinStretchCorn/Assets/Audio/Settings/AS_TwinStretchCorn_Short
 bOverrideConcurrency=true
 maxDistance=6400
 soundClassObject=&Fort_WeaponRanged_Fire_1P_PSM
 volumeMultiplier=1.2
 concurrencyOverrides {maxCount=1}
}
