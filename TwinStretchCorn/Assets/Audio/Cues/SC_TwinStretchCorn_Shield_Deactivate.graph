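# Sound cue for the TwinStretchCorn shield deactivation sound, using the shield's attenuation settings.
# A quality-level switch feeds a short fixed delay, then a modulator at the root.
# The first two quality slots randomly pick from five wave variants, while the last two pick from a three-wave subset.
# The five-wave random node is excluded from branch culling.
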SoundCue /TwinStretchCorn/Assets/Audio/Cues/SC_TwinStretchCorn_Shield_Deactivate
root=Modulator_1
nodes [
 {id=Delay_0, children [QualityLevel_1], properties {delayMax=0.05, delayMin=0.05}}
 {id=Modulator_1, children [Delay_0]}
 {id=QualityLevel_1, children [Random_1,Random_1,Random_0,Random_0]}
 {id=Random_0, children [WavePlayer_3,WavePlayer_8,WavePlayer_1], properties {weights [1,1,1]}}
 {
  id=Random_1
  children [WavePlayer_0,WavePlayer_4,WavePlayer_5,WavePlayer_6,WavePlayer_7]
  properties {bShouldExcludeFromBranchCulling=true, weights [1,1,1,1,1]}
 }
 {id=WavePlayer_0, properties {soundWaveAssetPtr=/TwinStretchCorn/Assets/Audio/Waves/TwinStretchCorn_Shield_Deactivate_01}}
 {id=WavePlayer_1, properties {soundWaveAssetPtr=/TwinStretchCorn/Assets/Audio/Waves/TwinStretchCorn_Shield_Deactivate_05}}
 {id=WavePlayer_3, properties {soundWaveAssetPtr=&TwinStretchCorn_Shield_Deactivate_01}}
 {id=WavePlayer_4, properties {soundWaveAssetPtr=/TwinStretchCorn/Assets/Audio/Waves/TwinStretchCorn_Shield_Deactivate_02}}
 {id=WavePlayer_5, properties {soundWaveAssetPtr=/TwinStretchCorn/Assets/Audio/Waves/TwinStretchCorn_Shield_Deactivate_03}}
 {id=WavePlayer_6, properties {soundWaveAssetPtr=/TwinStretchCorn/Assets/Audio/Waves/TwinStretchCorn_Shield_Deactivate_04}}
 {id=WavePlayer_7, properties {soundWaveAssetPtr=&TwinStretchCorn_Shield_Deactivate_05}}
 {id=WavePlayer_8, properties {soundWaveAssetPtr=&TwinStretchCorn_Shield_Deactivate_02}}
]
properties {
 attenuationSettings=/TwinStretchCorn/Assets/Audio/Settings/AS_TwinStretchCorn_Shield
 maxDistance=9400
 soundClassObject=&Fort_WeaponRanged_Fire_1P_PSM
 volumeMultiplier=0.25
 concurrencyOverrides {maxCount=3}
}
