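# Sound cue for the TwinStretchCorn shield forming, playing a randomly chosen Shield_Activate wave through a Modulator.
# A quality-level node routes the first two levels to a six-wave random pool and the last two to a three-wave subset, with only the larger pool excluded from branch culling.
# It uses shield attenuation, a first-person ranged-weapon fire sound class and a concurrency override capping simultaneous instances.
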SoundCue /TwinStretchCorn/Assets/Audio/Cues/SC_TwinStretchCorn_Shield_Form
root=Modulator_1
nodes [
 {id=Modulator_1, children [QualityLevel_0]}
 {id=QualityLevel_0, children [Random_1,Random_1,Random_0,Random_0]}
 {id=Random_0, children [WavePlayer_3,WavePlayer_2,WavePlayer_1], properties {weights [1,1,1]}}
 {
  id=Random_1
  children [WavePlayer_0,WavePlayer_4,WavePlayer_5,WavePlayer_6,WavePlayer_7,WavePlayer_8]
  properties {bShouldExcludeFromBranchCulling=true, weights [1,1,1,1,1,1]}
 }
 {id=WavePlayer_0, properties {soundWaveAssetPtr=/TwinStretchCorn/Assets/Audio/Waves/TwinStretchCorn_Shield_Activate_01}}
 {id=WavePlayer_1, properties {soundWaveAssetPtr=/TwinStretchCorn/Assets/Audio/Waves/TwinStretchCorn_Shield_Activate_05}}
 {id=WavePlayer_2, properties {soundWaveAssetPtr=/TwinStretchCorn/Assets/Audio/Waves/TwinStretchCorn_Shield_Activate_04}}
 {id=WavePlayer_3, properties {soundWaveAssetPtr=&TwinStretchCorn_Shield_Activate_01}}
 {id=WavePlayer_4, properties {soundWaveAssetPtr=/TwinStretchCorn/Assets/Audio/Waves/TwinStretchCorn_Shield_Activate_02}}
 {id=WavePlayer_5, properties {soundWaveAssetPtr=/TwinStretchCorn/Assets/Audio/Waves/TwinStretchCorn_Shield_Activate_03}}
 {id=WavePlayer_6, properties {soundWaveAssetPtr=&TwinStretchCorn_Shield_Activate_04}}
 {id=WavePlayer_7, properties {soundWaveAssetPtr=&TwinStretchCorn_Shield_Activate_05}}
 {id=WavePlayer_8, properties {soundWaveAssetPtr=/TwinStretchCorn/Assets/Audio/Waves/TwinStretchCorn_Shield_Activate_06}}
]
properties {
 attenuationSettings=/TwinStretchCorn/Assets/Audio/Settings/AS_TwinStretchCorn_Shield
 bOverrideConcurrency=true
 maxDistance=9400
 soundClassObject=&Fort_WeaponRanged_Fire_1P_PSM
 volumeMultiplier=0.8
 concurrencyOverrides {maxCount=3}
}
