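# Sound cue for TwinStretchCorn shield impacts on vehicles, spatialized with the shield's attenuation settings.
# A modulator feeds a quality-level switch: the first two slots randomly pick from seven impact waves, the last two from a subset of three.
# Concurrency allows one instance per owner and blocks new plays while it sounds.
# It is routed through a third-person ranged weapon fire sound class.
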SoundCue /TwinStretchCorn/Assets/Audio/Cues/SC_TwinStretchCorn_Shield_Impacts_Vehicle
root=Modulator_0
nodes [
 {id=Modulator_0, children [QualityLevel_0]}
 {id=QualityLevel_0, children [Random_2,Random_2,Random_3,Random_3]}
 {
  id=Random_2
  children [WavePlayer_10, WavePlayer_11, WavePlayer_12, WavePlayer_13, WavePlayer_14, WavePlayer_15, WavePlayer_16]
  properties {bShouldExcludeFromBranchCulling=true, weights [1,1,1,1,1,1,1]}
 }
 {id=Random_3, children [WavePlayer_19, WavePlayer_18, WavePlayer_17], properties {weights [1,1,1]}}
 {id=WavePlayer_10, properties {soundWaveAssetPtr=/TwinStretchCorn/Assets/Audio/Waves/TwinStretchCorn_Shield_Impacts_Vehicle_01}}
 {id=WavePlayer_11, properties {soundWaveAssetPtr=/TwinStretchCorn/Assets/Audio/Waves/TwinStretchCorn_Shield_Impacts_Vehicle_02}}
 {id=WavePlayer_12, properties {soundWaveAssetPtr=/TwinStretchCorn/Assets/Audio/Waves/TwinStretchCorn_Shield_Impacts_Vehicle_03}}
 {id=WavePlayer_13, properties {soundWaveAssetPtr=/TwinStretchCorn/Assets/Audio/Waves/TwinStretchCorn_Shield_Impacts_Vehicle_04}}
 {id=WavePlayer_14, properties {soundWaveAssetPtr=/TwinStretchCorn/Assets/Audio/Waves/TwinStretchCorn_Shield_Impacts_Vehicle_05}}
 {id=WavePlayer_15, properties {soundWaveAssetPtr=/TwinStretchCorn/Assets/Audio/Waves/TwinStretchCorn_Shield_Impacts_Vehicle_06}}
 {id=WavePlayer_16, properties {soundWaveAssetPtr=/TwinStretchCorn/Assets/Audio/Waves/TwinStretchCorn_Shield_Impacts_Vehicle_07}}
 {id=WavePlayer_17, properties {soundWaveAssetPtr=&TwinStretchCorn_Shield_Impacts_Vehicle_06}}
 {id=WavePlayer_18, properties {soundWaveAssetPtr=&TwinStretchCorn_Shield_Impacts_Vehicle_02}}
 {id=WavePlayer_19, properties {soundWaveAssetPtr=&TwinStretchCorn_Shield_Impacts_Vehicle_01}}
]
properties {
 attenuationSettings=/TwinStretchCorn/Assets/Audio/Settings/AS_TwinStretchCorn_Shield
 bOverrideConcurrency=true
 maxDistance=9400
 soundClassObject=&Fort_WeaponRanged_Fire_3P_PSM
 volumeMultiplier=0.6
 concurrencyOverrides {bLimitToOwner=true, maxCount=1, resolutionRule=PreventNew, volumeScaleAttackTime=0.1}
}
